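# UI material for the Item Shop navigation arrow, with a hover state that changes its color and size.
# A Rotation parameter turns the UVs by a fraction, then they scale about the center between ScaleStart and ScaleStop as Hover rises.
# The arrow texture's green channel is offset, multiplied by Softness and saturated into an adjustable-edge mask, distance-field style.
# Hover also blends Color to Color_Hover.
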Material /Game/UI/Frontend/ItemShop/Art/M_UI_ItemShop_Navigation_Arrow
graph {
 properties {
  bUsedWithStaticMesh=false
  blendMode=BLEND_Translucent
  materialDomain=MD_UI
  pixelDepthOffsetMode=PDOM_Legacy
  usedShadingModels=Default Lit
 }
 nodes [
  LinearInterpolate(
   A=VectorParameter() {parameterName=Color, defaultValue=<rgba 1 0 0 1>}.RGB
   B=VectorParameter() {parameterName=Color_Hover, defaultValue=<rgba 0 1 0 1>}.RGB
   Alpha=ScalarParameter() {parameterName=Hover}
  )
  n1 = ScalarParameter() {parameterName=Rotation}
  n2 = TextureCoordinate()
  n3 = ScalarParameter() {defaultValue=0.9, parameterName=ScaleStart}
  n4 = ScalarParameter() {defaultValue=1, parameterName=ScaleStop}
  n5 = ScalarParameter() {parameterName=Hover}
  n6 = TextureSample(UVs=MaterialFunctionCall(
    UVs=MaterialFunctionCall!(Rotation_Fraction=@n1, TexCoord=@n2) {materialFunction=/FortUILibrary/Foundation/Art_ToolKit/AngledBoxes/MF_UI_Rotator_Fractions}.Result
    Texture Scale=LinearInterpolate(A=@n3, B=@n4, Alpha=@n5)
   ) {materialFunction=&ScaleUVsByCenter}.UVs) {samplerSource=SSM_Clamp_WorldGroupSettings, texture=/Game/UI/Frontend/Showdown/Materials/T_UI_Arrow_DF}
  n7 = ScalarParameter() {defaultValue=50, parameterName=Softness}
  Saturate(None=Multiply(A=Subtract(A=@n6.G) {constB=0.125}, B=Multiply(A=@n7) {constB=0.1}))
 ]
}
