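# Translucent UI material for an Item Shop section pip, drawing a flat-colored shape with slightly angled top and bottom edges.
# Color blends from ColorA to ColorB by Hover, then to ColorC by Selected.
# The alpha mask multiplies two offset gradients from rotated UV V, one inverted, then scales by FinalAlpha.
# Edge gradients are steepened by Sharpness and a resolution-correction function, giving crisp procedural edges without textures.
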Material /Game/UI/Frontend/ItemShop/Art/M_UI_ItemShop_SectionPip
graph {
 properties {
  bUsedWithStaticMesh=false
  blendMode=BLEND_Translucent
  materialDomain=MD_UI
  pixelDepthOffsetMode=PDOM_Legacy
  usedShadingModels=Default Lit
 }
 nodes [
  LinearInterpolate(
   A=LinearInterpolate(
    A=VectorParameter() {parameterName=ColorA, defaultValue=<rgba 1 0.877 0 1>}.RGB
    B=VectorParameter() {parameterName=ColorB, defaultValue=<rgba 0 1 0 1>}.RGB
    Alpha=ScalarParameter() {parameterName=Hover}
   )
   B=VectorParameter() {parameterName=ColorC, defaultValue=<rgba 0 1 0.9833 1>}.RGB
   Alpha=ScalarParameter() {parameterName=Selected}
  )
  n1 = ScalarParameter() {defaultValue=-0.0075, parameterName=Rotation}
  n2 = TextureCoordinate()
  n3 = ComponentMask(None=Add(
    A=MaterialFunctionCall!(Rotation_Fraction=@n1, TexCoord=@n2) {materialFunction=/FortUILibrary/Foundation/Art_ToolKit/AngledBoxes/MF_UI_Rotator_Fractions}.Result
    B=ScalarParameter() {defaultValue=-0.025, parameterName=TopOffset}
   )) {g=true}
  n4 = Multiply(
   A=ScalarParameter() {defaultValue=64, parameterName=Sharpness}
   B=MaterialFunctionCall!(Native_U_PixelSize=ScalarParameter() {defaultValue=16, parameterName=PixelSize_U}) {materialFunction=/FortUILibrary/Foundation/Art_ToolKit/Tools/M_UI_Resolution_Sharpness_Correction}.Result
  )
  n5 = MaterialFunctionCall(Rotation_Fraction=ScalarParameter() {defaultValue=-0.0075, parameterName=Rotation}, TexCoord=@n2) {materialFunction=&MF_UI_Rotator_Fractions}
  n6 = Multiply(A=ScalarParameter() {defaultValue=-0.025, parameterName=Offset}) {constB=-1}
  n7 = ComponentMask(None=OneMinus(None=Add(A=@n5.Result, B=@n6))) {g=true}
  n8 = DotProduct(A=Saturate(None=Multiply(A=@n3, B=@n4)), B=Saturate(None=Multiply(A=@n4, B=@n7)))
  Saturate(None=Multiply(A=Saturate(None=@n8), B=ScalarParameter() {defaultValue=1, parameterName=FinalAlpha}))
 ]
}
